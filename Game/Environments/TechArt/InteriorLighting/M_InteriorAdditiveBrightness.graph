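# Unlit additive material for interior lighting that adds a tinted brightness, set by AdditiveColor, without ray-traced shadows.
# Camera distance to each instance's origin drives a near fade through NearFadeOffset and NearFadeBlend, which multiplies a DepthFade to soften intersections with scene geometry.
# When the near fade reaches zero, an If node yields a vector from each vertex to the instance origin, an offset that would collapse the mesh.
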Material /Game/Environments/TechArt/InteriorLighting/M_InteriorAdditiveBrightness
graph {
 properties {bAutomaticallySetUsageInEditor=false, bCastRayTracedShadows=false, blendMode=BLEND_Additive, shadingModel=MSM_Unlit, usedShadingModels=Unlit}
 nodes [
  Constant3Vector()
  n1 = NamedRerouteDeclaration(None=TransformPosition(None=Constant3Vector()) {transformSourceType=TRANSFORMPOSSOURCE_Instance, transformType=TRANSFORMPOSSOURCE_World}) {name=OriginPoint, nodeColor=<rgba 1 0.4453 0 1>}
  n2 = Distance(A=CameraPositionWS(), B=@n1)
  n3 = ScalarParameter() {defaultValue=2000, parameterName=NearFadeOffset}
  n4 = Saturate(None=Divide(A=Subtract(A=@n2, B=@n3), B=ScalarParameter() {defaultValue=4500, parameterName=NearFadeBlend}))
  n5 = Subtract(A=@n1, B=WorldPosition().XYZ)
  If(A=@n4, "A > B"=Constant() {r=1}, "A == B"=@n5, "A < B"=@n5)
  n6 = ScalarParameter() {defaultValue=1506.1239, parameterName=DepthDistance}
  Saturate(None=Multiply(A=DepthFade(FadeDistance=@n6), B=@n4))
  VectorParameter() {parameterName=AdditiveColor, defaultValue=<rgba 0.1615 0.0323 0.1301 1>}
 ]
}
